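# Physics asset for the necklace of the Daisy female medium character body, for secondary motion of the necklace against the upper body.
# A three-capsule dyn_necklace chain hangs from base_necklace, which is angularly locked to spine_05, with the swing2 angle setting widening down the chain.
# Two capsules on spine_05 and one on each pec aim bone form unconstrained or anchor body volumes approximating the chest.
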PhysicsAsset /BRCosmetics/Characters/Player/Female/Medium/Bodies/F_MED_LocalTalent_Daisy/Meshes/F_MED_LocalTalent_DaisyNecklace_Physics
bodies {
 base_necklace [{shapeType=Capsule, length=0.505, radius=0.505, shapeName=base_necklace_capsule, rotation=<rot 0 -90 0>}]
 deform_pec_aim_l [{shapeType=Capsule, length=6, radius=4.2, shapeName=deform_pec_aim_l_capsule, center=<2.8807 -0.5392 -3.4663>, rotation=<rot 0 88 -0.5179>}]
 deform_pec_aim_r [{
   shapeType=Capsule
   length=5.9214
   radius=4.2
   shapeName=deform_pec_aim_r_capsule
   center=<-2.9055 -0.1301 3.5048>
   rotation=<rot 0 87.7669 11.2376>
  }]
 dyn_necklace_01 [{shapeType=Capsule, length=5, radius=2, shapeName=dyn_necklace_01_capsule, center=<2 0 2>, rotation=<rot 0 -90.3858 -90>}]
 dyn_necklace_02 [{shapeType=Capsule, length=4, radius=3, shapeName=dyn_necklace_02_capsule, center=<3 0 2>, rotation=<rot 0 -90 -98.1336>}]
 dyn_necklace_03 [{shapeType=Capsule, length=4, radius=3.5, shapeName=dyn_necklace_03_capsule, center=<4 0 2.8>, rotation=<rot 0 -90 -88>}]
 spine_05 [
  {shapeType=Capsule, length=9.7199, radius=11, shapeName=spine_05_capsule, center=<-5 -0.4133 -0.0222>}
  {
   shapeType=Capsule
   length=19.3639
   radius=8.0064
   shapeName=spine_05_capsule
   center=<-6.6308 4.6414 -2.7366>
   rotation=<rot 3.0682 -14.7938 -3.8016>
  }
 ]
}
constraints [
 {
  bone1Name=base_necklace
  bone2Name=spine_05
  swing1LimitDegrees=45
  swing1Motion=Locked
  swing2LimitDegrees=45
  swing2Motion=Locked
  twistLimitDegrees=45
  twistMotion=Locked
 }
 {bone1Name=dyn_necklace_01, bone2Name=base_necklace, swing1LimitDegrees=10, swing2LimitDegrees=20, twistLimitDegrees=0}
 {bone1Name=dyn_necklace_02, bone2Name=dyn_necklace_01, swing1LimitDegrees=10, swing2LimitDegrees=30, twistLimitDegrees=0}
 {bone1Name=dyn_necklace_03, bone2Name=dyn_necklace_02, swing1LimitDegrees=10, swing2LimitDegrees=40, twistLimitDegrees=0}
]
