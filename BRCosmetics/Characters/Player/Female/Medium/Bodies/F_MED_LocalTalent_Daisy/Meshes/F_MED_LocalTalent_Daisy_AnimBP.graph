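# Cosmetic body AnimBP for the Daisy character that copies the parent mesh pose and adds secondary motion.
# AnimDynamics springs drive loop bones, some with pelvis or thigh collision spheres, before a Control Rig corrects for movement and stance states.
# A RigidBody necklace sim and further springs for charm, sleeves, ribbons and knots follow.
# The EventGraph computes crouch and strafe inputs, plus gravity override values the AnimGraph never reads.
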AnimBlueprint /BRCosmetics/Characters/Player/Female/Medium/Bodies/F_MED_LocalTalent_Daisy/Meshes/F_MED_LocalTalent_Daisy_AnimBP
target_skeleton=&/F_MED_LocalTalent_Daisy_Skeleton
graphs [
 AnimGraph {nodes [
   n1 = Anim Dynamics - Bone: dyn_loop_b_l_01!(
    ComponentPose=Local To Component(LocalPose=Copy Pose From Mesh() {bUseAttachedParent=true})
    Alpha=1
    GravityOverride:Vector="0,0,-980"
    ExternalForce="0,0,0"
   ) -> (Pose) {
    angularDampingOverride=1
    angularSpringConstant=7000
    bAngularSpring=true
    bOverrideAngularDamping=true
    bUseGravityOverride=true
    gravityScale=0.35
    simulationSpace=RootRelative
    boundBone {boneName=dyn_loop_b_l_01}
    componentAppliedLinearAccClamp=<5000 5000 5000>
    componentLinearAccScale=<1 1 1.5>
    componentLinearVelScale=<1 1 1>
    gravityOverride {z=-980}
    physicsBodyDefinitions [{
      boundBone {boneName=dyn_loop_b_l_01}
      boxExtents=<20 30 30>
      constraintSetup {angularLimitsMax=<0 2 30>, angularLimitsMin=<0 -35 -30>, angularTarget=<1 0 0>}
      localJointOffset=<3 0 0>
     }]
    simSpaceSettings {simSpaceAngularAlpha=0.8}
   }
   n2 = Anim Dynamics - Bone: dyn_loop_b_l_02!(ComponentPose=@n1, Alpha=1, GravityOverride:Vector="0,0,-980", ExternalForce="0,0,0") -> (Pose) {
    angularDampingOverride=1
    angularSpringConstant=7000
    bAngularSpring=true
    bOverrideAngularDamping=true
    bUseGravityOverride=true
    bUseSphericalLimits=true
    gravityScale=0.35
    simulationSpace=RootRelative
    boundBone {boneName=dyn_loop_b_l_02}
    componentAppliedLinearAccClamp=<5000 5000 5000>
    componentLinearAccScale=<1 1 1.5>
    componentLinearVelScale=<1 1 1>
    gravityOverride {z=-980}
    physicsBodyDefinitions [{
      boundBone {boneName=dyn_loop_b_l_02}
      boxExtents=<20 30 30>
      constraintSetup {angularLimitsMax=<0 2 30>, angularLimitsMin=<0 -35 -30>, angularTarget=<1 0 0>}
      localJointOffset=<3 0 0>
     }]
    simSpaceSettings {simSpaceAngularAlpha=0.8}
    sphericalLimits [{limitRadius=12.478, drivingBone {boneName=pelvis}, sphereLocalOffset=<-4.0736 -0.7074 -2.7108>}]
   }
   n3 = Anim Dynamics - Bone: dyn_loop_b_l_03!(ComponentPose=@n2, Alpha=1, GravityOverride:Vector="0,0,-980", ExternalForce="0,0,0") -> (Pose) {
    angularDampingOverride=1
    angularSpringConstant=7000
    bAngularSpring=true
    bOverrideAngularDamping=true
    bUseGravityOverride=true
    gravityScale=0.35
    simulationSpace=RootRelative
    boundBone {boneName=dyn_loop_b_l_03}
    componentAppliedLinearAccClamp=<5000 5000 5000>
    componentLinearAccScale=<1 1 1.5>
    componentLinearVelScale=<1 1 1>
    gravityOverride {z=-980}
    physicsBodyDefinitions [{
      boundBone {boneName=dyn_loop_b_l_03}
      boxExtents=<20 30 30>
      constraintSetup {angularLimitsMax=<0 2 30>, angularLimitsMin=<0 -35 -30>, angularTarget=<1 0 0>}
      localJointOffset=<3 0 0>
     }]
    simSpaceSettings {simSpaceAngularAlpha=0.8}
   }
   n4 = Anim Dynamics - Bone: dyn_loop_b_l_04!(ComponentPose=@n3, Alpha=1, GravityOverride:Vector="0,0,-980", ExternalForce="0,0,0") -> (Pose) {
    angularDampingOverride=1
    angularSpringConstant=7000
    bAngularSpring=true
    bOverrideAngularDamping=true
    bUseGravityOverride=true
    bUseSphericalLimits=true
    gravityScale=0.35
    simulationSpace=RootRelative
    boundBone {boneName=dyn_loop_b_l_04}
    componentAppliedLinearAccClamp=<5000 5000 5000>
    componentLinearAccScale=<1 1 1.5>
    componentLinearVelScale=<1 1 1>
    gravityOverride {z=-980}
    physicsBodyDefinitions [{
      boundBone {boneName=dyn_loop_b_l_04}
      boxExtents=<20 30 30>
      constraintSetup {angularLimitsMax=<0 2 30>, angularLimitsMin=<0 -35 -30>, angularTarget=<1 0 0>}
      localJointOffset=<3 0 0>
     }]
    simSpaceSettings {simSpaceAngularAlpha=0.8}
    sphericalLimits [{limitRadius=8.517, drivingBone {boneName=thigh_twist_01_l}, sphereLocalOffset=<4.5079 -0.4525 -9.4406>}]
   }
   n5 = Anim Dynamics - Bone: dyn_loop_a_l_01!(ComponentPose=@n4, Alpha=1, GravityOverride:Vector="0,0,-980", ExternalForce="0,0,0") -> (Pose) {
    angularDampingOverride=1
    angularSpringConstant=7000
    bAngularSpring=true
    bOverrideAngularDamping=true
    bUseGravityOverride=true
    bUseSphericalLimits=true
    gravityScale=0.35
    simulationSpace=RootRelative
    boundBone {boneName=dyn_loop_a_l_01}
    componentAppliedLinearAccClamp=<5000 5000 5000>
    componentLinearAccScale=<1 1 1.5>
    componentLinearVelScale=<1 1 1>
    gravityOverride {z=-980}
    physicsBodyDefinitions [{
      boundBone {boneName=dyn_loop_a_l_01}
      boxExtents=<20 30 30>
      constraintSetup {angularLimitsMax=<0 2 30>, angularLimitsMin=<0 -35 -30>, angularTarget=<1 0 0>}
      localJointOffset=<3 0 0>
     }]
    simSpaceSettings {simSpaceAngularAlpha=0.8}
    sphericalLimits [{limitRadius=10.9199, drivingBone {boneName=pelvis}, sphereLocalOffset=<-5.0223 -2.4549 -4.6159>}]
   }
   n6 = Anim Dynamics - Bone: dyn_loop_a_l_02!(ComponentPose=@n5, Alpha=1, GravityOverride:Vector="0,0,-980", ExternalForce="0,0,0") -> (Pose) {
    angularDampingOverride=1
    angularSpringConstant=7000
    bAngularSpring=true
    bOverrideAngularDamping=true
    bUseGravityOverride=true
    gravityScale=0.35
    simulationSpace=RootRelative
    boundBone {boneName=dyn_loop_a_l_02}
    componentAppliedLinearAccClamp=<5000 5000 5000>
    componentLinearAccScale=<1 1 1.5>
    componentLinearVelScale=<1 1 1>
    gravityOverride {z=-980}
    physicsBodyDefinitions [{
      boundBone {boneName=dyn_loop_a_l_02}
      boxExtents=<20 30 30>
      constraintSetup {angularLimitsMax=<0 2 30>, angularLimitsMin=<0 -35 -30>, angularTarget=<1 0 0>}
      localJointOffset=<3 0 0>
     }]
    simSpaceSettings {simSpaceAngularAlpha=0.8}
   }
   n7 = Anim Dynamics - Bone: dyn_loop_a_l_03!(ComponentPose=@n6, Alpha=1, GravityOverride:Vector="0,0,-980", ExternalForce="0,0,0") -> (Pose) {
    angularDampingOverride=1
    angularSpringConstant=7000
    bAngularSpring=true
    bOverrideAngularDamping=true
    bUseGravityOverride=true
    gravityScale=0.35
    simulationSpace=RootRelative
    boundBone {boneName=dyn_loop_a_l_03}
    componentAppliedLinearAccClamp=<5000 5000 5000>
    componentLinearAccScale=<1 1 1.5>
    componentLinearVelScale=<1 1 1>
    gravityOverride {z=-980}
    physicsBodyDefinitions [{
      boundBone {boneName=dyn_loop_a_l_03}
      boxExtents=<20 30 30>
      constraintSetup {angularLimitsMax=<0 2 30>, angularLimitsMin=<0 -35 -30>, angularTarget=<1 0 0>}
      localJointOffset=<3 0 0>
     }]
    simSpaceSettings {simSpaceAngularAlpha=0.8}
   }
   n8 = Anim Dynamics - Bone: dyn_loop_a_r_01!(ComponentPose=@n7, Alpha=1, GravityOverride:Vector="0,0,-980", ExternalForce="0,0,0") -> (Pose) {
    angularDampingOverride=1
    angularSpringConstant=7000
    bAngularSpring=true
    bOverrideAngularDamping=true
    bUseGravityOverride=true
    gravityScale=0.35
    simulationSpace=RootRelative
    boundBone {boneName=dyn_loop_a_r_01}
    componentAppliedLinearAccClamp=<5000 5000 5000>
    componentLinearAccScale=<0.5 0.5 0.5>
    componentLinearVelScale=<0.5 5 0.5>
    gravityOverride {z=-980}
    physicsBodyDefinitions [{
      boundBone {boneName=dyn_loop_a_r_01}
      boxExtents=<20 30 30>
      constraintSetup {angularLimitsMax=<0 35 30>, angularLimitsMin=<0 -2 -30>, angularTarget=<1 0 0>}
      localJointOffset=<3 0 0>
     }]
    simSpaceSettings {simSpaceAngularAlpha=0.8}
   }
   n9 = Anim Dynamics - Bone: dyn_loop_a_r_02!(ComponentPose=@n8, Alpha=1, GravityOverride:Vector="0,0,-980", ExternalForce="0,0,0") -> (Pose) {
    angularDampingOverride=1
    angularSpringConstant=7000
    bAngularSpring=true
    bOverrideAngularDamping=true
    bUseGravityOverride=true
    bUseSphericalLimits=true
    gravityScale=0.35
    simulationSpace=RootRelative
    boundBone {boneName=dyn_loop_a_r_02}
    componentAppliedLinearAccClamp=<5000 5000 5000>
    componentLinearAccScale=<0.5 0.5 0.5>
    componentLinearVelScale=<0.5 0.5 0.5>
    gravityOverride {z=-980}
    physicsBodyDefinitions [{
      boundBone {boneName=dyn_loop_a_r_02}
      boxExtents=<20 30 30>
      constraintSetup {angularLimitsMax=<0 35 30>, angularLimitsMin=<0 -2 -30>, angularTarget=<1 0 0>}
      localJointOffset=<3 0 0>
     }]
    simSpaceSettings {simSpaceAngularAlpha=0.8}
    sphericalLimits [{limitRadius=11.6332, drivingBone {boneName=pelvis}, sphereLocalOffset=<-5.4019 -2.1973 4.7456>}]
   }
   n10 = Anim Dynamics - Bone: dyn_loop_a_r_03!(ComponentPose=@n9, Alpha=1, GravityOverride:Vector="0,0,-980", ExternalForce="0,0,0") -> (Pose) {
    angularDampingOverride=1
    angularSpringConstant=7000
    bAngularSpring=true
    bOverrideAngularDamping=true
    bUseGravityOverride=true
    bUseSphericalLimits=true
    gravityScale=0.35
    simulationSpace=RootRelative
    boundBone {boneName=dyn_loop_a_r_03}
    componentAppliedLinearAccClamp=<5000 5000 5000>
    componentLinearAccScale=<0.5 0.5 0.5>
    componentLinearVelScale=<0.5 5 0.5>
    gravityOverride {z=-980}
    physicsBodyDefinitions [{
      boundBone {boneName=dyn_loop_a_r_03}
      boxExtents=<20 30 30>
      constraintSetup {angularLimitsMax=<0 35 30>, angularLimitsMin=<0 -2 -30>, angularTarget=<1 0 0>}
      localJointOffset=<3 0 0>
     }]
    simSpaceSettings {simSpaceAngularAlpha=0.8}
    sphericalLimits [{limitRadius=11.6332, drivingBone {boneName=pelvis}, sphereLocalOffset=<-5.4019 -2.1973 4.7456>}]
   }
   n11 = Anim Dynamics - Bone: dyn_loop_a_r_04!(ComponentPose=@n10, Alpha=1, GravityOverride:Vector="0,0,-980", ExternalForce="0,0,0") -> (Pose) {
    angularDampingOverride=1
    angularSpringConstant=7000
    bAngularSpring=true
    bOverrideAngularDamping=true
    bUseGravityOverride=true
    bUseSphericalLimits=true
    gravityScale=0.35
    simulationSpace=RootRelative
    boundBone {boneName=dyn_loop_a_r_04}
    componentAppliedLinearAccClamp=<5000 5000 5000>
    componentLinearAccScale=<0.5 0.5 0.5>
    componentLinearVelScale=<0.5 5 0.5>
    gravityOverride {z=-980}
    physicsBodyDefinitions [{
      boundBone {boneName=dyn_loop_a_r_04}
      boxExtents=<20 30 30>
      constraintSetup {angularLimitsMax=<0 35 30>, angularLimitsMin=<0 -2 -30>, angularTarget=<1 0 0>}
      localJointOffset=<3 0 0>
     }]
    simSpaceSettings {simSpaceAngularAlpha=0.8}
    sphericalLimits [{limitRadius=11.6332, drivingBone {boneName=pelvis}, sphereLocalOffset=<-5.4019 -2.1973 4.7456>}]
   }
   n12 = Anim Dynamics - Bone: dyn_loop_b_r_01!(ComponentPose=@n11, Alpha=1, GravityOverride:Vector="0,0,-980", ExternalForce="0,0,0") -> (Pose) {
    angularDampingOverride=1
    angularSpringConstant=7000
    bAngularSpring=true
    bOverrideAngularDamping=true
    bUseGravityOverride=true
    bUseSphericalLimits=true
    gravityScale=0.35
    simulationSpace=RootRelative
    boundBone {boneName=dyn_loop_b_r_01}
    componentAppliedLinearAccClamp=<5000 5000 5000>
    componentLinearAccScale=<0.5 0.5 0.5>
    componentLinearVelScale=<0.5 5 0.5>
    gravityOverride {z=-980}
    physicsBodyDefinitions [{
      boundBone {boneName=dyn_loop_b_r_01}
      boxExtents=<20 30 30>
      constraintSetup {angularLimitsMax=<0 35 30>, angularLimitsMin=<0 -2 -30>, angularTarget=<1 0 0>}
      localJointOffset=<3 0 0>
     }]
    simSpaceSettings {simSpaceAngularAlpha=0.8}
    sphericalLimits [{limitRadius=12.6657, drivingBone {boneName=pelvis}, sphereLocalOffset=<-5.6541 0.0169 3.7259>}]
   }
   n13 = Anim Dynamics - Bone: dyn_loop_b_r_02!(ComponentPose=@n12, Alpha=1, GravityOverride:Vector="0,0,-980", ExternalForce="0,0,0") -> (Pose) {
    angularDampingOverride=1
    angularSpringConstant=7000
    bAngularSpring=true
    bOverrideAngularDamping=true
    bUseGravityOverride=true
    gravityScale=0.35
    simulationSpace=RootRelative
    boundBone {boneName=dyn_loop_b_r_02}
    componentAppliedLinearAccClamp=<5000 5000 5000>
    componentLinearAccScale=<0.5 0.5 0.5>
    componentLinearVelScale=<0.5 5 0.5>
    gravityOverride {z=-980}
    physicsBodyDefinitions [{
      boundBone {boneName=dyn_loop_b_r_02}
      boxExtents=<20 30 30>
      constraintSetup {angularLimitsMax=<0 35 30>, angularLimitsMin=<0 -2 -30>, angularTarget=<1 0 0>}
      localJointOffset=<3 0 0>
     }]
    simSpaceSettings {simSpaceAngularAlpha=0.8}
   }
   n14 = Anim Dynamics - Bone: dyn_loop_b_r_03!(ComponentPose=@n13, Alpha=1, GravityOverride:Vector="0,0,-980", ExternalForce="0,0,0") -> (Pose) {
    angularDampingOverride=1
    angularSpringConstant=7000
    bAngularSpring=true
    bOverrideAngularDamping=true
    bUseGravityOverride=true
    gravityScale=0.35
    simulationSpace=RootRelative
    boundBone {boneName=dyn_loop_b_r_03}
    componentAppliedLinearAccClamp=<5000 5000 5000>
    componentLinearAccScale=<0.5 0.5 0.5>
    componentLinearVelScale=<0.5 5 0.5>
    gravityOverride {z=-980}
    physicsBodyDefinitions [{
      boundBone {boneName=dyn_loop_b_r_03}
      boxExtents=<20 30 30>
      constraintSetup {angularLimitsMax=<0 35 30>, angularLimitsMin=<0 -2 -30>, angularTarget=<1 0 0>}
      localJointOffset=<3 0 0>
     }]
    simSpaceSettings {simSpaceAngularAlpha=0.8}
   }
   n15 = Control Rig!(
    Source=Component To Local(ComponentPose=@n14)
    Alpha
    crouch:bool=Get bIsCrouching()
    MoveLeft:double="Get Move Left "() -> ("Move Left ":double)
    MoveRight:double=Get Move Right() -> (Move Right:double)
    BaseballSlide:bool=Get bIsBaseballSliding()
    KneeSlide:bool=Get bIsKneeSliding()
    Targeting:bool=Get bIsTargeting()
    PlayingEmote:bool=Get bIsPlayingEmote()
    onlycrouch:bool=Get OnlyCrouch() -> (OnlyCrouch:bool)
    CrouchStraf:double=Get isCrouchStrafe float() -> (isCrouchStrafe float:double)
    isInFrontEnd:bool=Get bIsInFrontEnd()
   ) -> (Pose) {controlRigAssetReference {blueprintRigClass=&/F_MED_LocalTalent_Daisy_CtrlRig.F_MED_LocalTalent_Daisy_CtrlRig_C}}
   n16 = RigidBody!(ComponentPose=Local To Component(LocalPose=@n15), Alpha, OverrideWorldGravity:Vector="0,0,-980", ExternalForce) -> (Pose) {
    bOverrideWorldGravity=true
    overridePhysicsAsset=&/F_MED_LocalTalent_DaisyNecklace_Physics
    simulationSpace=BaseBoneSpace
    baseBoneRef {boneName=spine_03}
    simSpaceSettings {dampingAlpha=0, maxAngularAcceleration=20, maxAngularVelocity=8, maxLinearAcceleration=1500, maxLinearVelocity=500, worldAlpha=1}
   }
   n17 = Anim Dynamics - Bone: dyn_charm!(ComponentPose=@n16, Alpha=1, GravityOverride:Vector="0,0,-980", ExternalForce="0,0,0") -> (Pose) {
    angularDampingOverride=1
    angularSpringConstant=15000
    bAngularSpring=true
    bOverrideAngularDamping=true
    bUseGravityOverride=true
    gravityScale=0.35
    simulationSpace=RootRelative
    boundBone {boneName=dyn_charm}
    componentAppliedLinearAccClamp=<5000 5000 5000>
    componentLinearAccScale=<1 2 1>
    componentLinearVelScale=<1 2 1>
    gravityOverride {z=-980}
    physicsBodyDefinitions [{
      boundBone {boneName=dyn_charm}
      boxExtents=<20 30 30>
      constraintSetup {angularLimitsMax=<0 0 20>, angularLimitsMin=<0 -40 -20>, angularTarget=<1 0 0>}
      localJointOffset=<6 0 0>
     }]
   }
   n18 = Anim Dynamics - Bone: dyn_TopSleeve_l(ComponentPose=@n17, Alpha=1, ExternalForce="0,0,0") {
    angularDampingOverride=1
    angularSpringConstant=12000
    bAngularSpring=true
    bOverrideAngularDamping=true
    bUseGravityOverride=true
    gravityScale=0.35
    simulationSpace=RootRelative
    boundBone {boneName=dyn_TopSleeve_l}
    componentAppliedLinearAccClamp=<3000 3000 3000>
    gravityOverride {z=-980}
    physicsBodyDefinitions [{
      boundBone {boneName=dyn_TopSleeve_l}
      boxExtents=<20 30 30>
      constraintSetup {angularLimitsMax=<0 2 10>, angularLimitsMin=<0 -2 -10>, angularTarget=<1 0 0>}
      localJointOffset=<2 0 0>
     }]
    simSpaceSettings {maxAngularAcceleration=14, maxAngularVelocity=7, simSpaceAngularAlpha=0.5}
   }
   n19 = Anim Dynamics - Bone: dyn_RightSleeve_l(ComponentPose=@n18, Alpha=1, ExternalForce="0,0,0") {
    angularDampingOverride=1
    angularSpringConstant=12000
    bAngularSpring=true
    bOverrideAngularDamping=true
    bUseGravityOverride=true
    gravityScale=0.35
    simulationSpace=RootRelative
    boundBone {boneName=dyn_RightSleeve_l}
    componentAppliedLinearAccClamp=<3000 3000 3000>
    gravityOverride {z=-980}
    physicsBodyDefinitions [{
      boundBone {boneName=dyn_RightSleeve_l}
      boxExtents=<20 30 30>
      constraintSetup {angularLimitsMax=<0 2 0>, angularLimitsMin=<0 -2 -10>, angularTarget=<1 0 0>}
      localJointOffset=<2 0 0>
     }]
    simSpaceSettings {maxAngularAcceleration=14, maxAngularVelocity=7, simSpaceAngularAlpha=0.5}
   }
   n20 = Anim Dynamics - Bone: dyn_LeftSleeve_l(ComponentPose=@n19, Alpha=1, ExternalForce="0,0,0") {
    angularDampingOverride=1
    angularSpringConstant=12000
    bAngularSpring=true
    bOverrideAngularDamping=true
    bUseGravityOverride=true
    gravityScale=0.35
    simulationSpace=RootRelative
    boundBone {boneName=dyn_LeftSleeve_l}
    componentAppliedLinearAccClamp=<3000 3000 3000>
    gravityOverride {z=-980}
    physicsBodyDefinitions [{
      boundBone {boneName=dyn_LeftSleeve_l}
      boxExtents=<20 30 30>
      constraintSetup {angularLimitsMax=<0 2 0>, angularLimitsMin=<0 -2 -10>, angularTarget=<1 0 0>}
      localJointOffset=<2 0 0>
     }]
    simSpaceSettings {maxAngularAcceleration=14, maxAngularVelocity=7, simSpaceAngularAlpha=0.5}
   }
   n21 = Anim Dynamics - Bone: dyn_BotSleeve_l(ComponentPose=@n20, Alpha=1, ExternalForce="0,0,0") {
    angularDampingOverride=1
    angularSpringConstant=12000
    bAngularSpring=true
    bOverrideAngularDamping=true
    bUseGravityOverride=true
    gravityScale=0.35
    simulationSpace=RootRelative
    boundBone {boneName=dyn_BotSleeve_l}
    componentAppliedLinearAccClamp=<3000 3000 3000>
    gravityOverride {z=-980}
    physicsBodyDefinitions [{
      boundBone {boneName=dyn_BotSleeve_l}
      boxExtents=<20 30 30>
      constraintSetup {angularLimitsMax=<0 2 2>, angularLimitsMin=<0 -2 -10>, angularTarget=<1 0 0>}
      localJointOffset=<2 0 0>
     }]
    simSpaceSettings {maxAngularAcceleration=14, maxAngularVelocity=7, simSpaceAngularAlpha=0.5}
   }
   n22 = Anim Dynamics - Bone: dyn_BotSleeve_r(ComponentPose=@n21, Alpha=1, ExternalForce="0,0,0") {
    angularDampingOverride=1
    angularSpringConstant=12000
    bAngularSpring=true
    bOverrideAngularDamping=true
    bUseGravityOverride=true
    gravityScale=0.35
    simulationSpace=RootRelative
    boundBone {boneName=dyn_BotSleeve_r}
    componentAppliedLinearAccClamp=<3000 3000 3000>
    gravityOverride {z=-980}
    physicsBodyDefinitions [{
      boundBone {boneName=dyn_BotSleeve_r}
      boxExtents=<20 30 30>
      constraintSetup {angularLimitsMax=<0 2 0>, angularLimitsMin=<0 -2 -10>, angularTarget=<1 0 0>}
      localJointOffset=<2 0 0>
     }]
    simSpaceSettings {maxAngularAcceleration=14, maxAngularVelocity=7, simSpaceAngularAlpha=0.5}
   }
   n23 = Anim Dynamics - Bone: dyn_TopSleeve_r(ComponentPose=@n22, Alpha=1, ExternalForce="0,0,0") {
    angularDampingOverride=1
    angularSpringConstant=12000
    bAngularSpring=true
    bOverrideAngularDamping=true
    bUseGravityOverride=true
    gravityScale=0.35
    simulationSpace=RootRelative
    boundBone {boneName=dyn_TopSleeve_r}
    componentAppliedLinearAccClamp=<3000 3000 3000>
    gravityOverride {z=-980}
    physicsBodyDefinitions [{
      boundBone {boneName=dyn_TopSleeve_r}
      boxExtents=<20 30 30>
      constraintSetup {angularLimitsMax=<0 2 0>, angularLimitsMin=<0 -2 -10>, angularTarget=<1 0 0>}
      localJointOffset=<2 0 0>
     }]
    simSpaceSettings {maxAngularAcceleration=14, maxAngularVelocity=7, simSpaceAngularAlpha=0.5}
   }
   n24 = Anim Dynamics - Bone: dyn_RightSleeve_r(ComponentPose=@n23, Alpha=1, ExternalForce="0,0,0") {
    angularDampingOverride=1
    angularSpringConstant=12000
    bAngularSpring=true
    bOverrideAngularDamping=true
    bUseGravityOverride=true
    gravityScale=0.35
    simulationSpace=RootRelative
    boundBone {boneName=dyn_RightSleeve_r}
    componentAppliedLinearAccClamp=<3000 3000 3000>
    gravityOverride {z=-980}
    physicsBodyDefinitions [{
      boundBone {boneName=dyn_RightSleeve_r}
      boxExtents=<20 30 30>
      constraintSetup {angularLimitsMax=<0 2 0>, angularLimitsMin=<0 -2 -10>, angularTarget=<1 0 0>}
      localJointOffset=<2 0 0>
     }]
    simSpaceSettings {maxAngularAcceleration=14, maxAngularVelocity=7, simSpaceAngularAlpha=0.5}
   }
   n25 = Anim Dynamics - Bone: dyn_LeftSleeve_r(ComponentPose=@n24, Alpha=1, ExternalForce="0,0,0") {
    angularDampingOverride=1
    angularSpringConstant=12000
    bAngularSpring=true
    bOverrideAngularDamping=true
    bUseGravityOverride=true
    gravityScale=0.35
    simulationSpace=RootRelative
    boundBone {boneName=dyn_LeftSleeve_r}
    componentAppliedLinearAccClamp=<3000 3000 3000>
    gravityOverride {z=-980}
    physicsBodyDefinitions [{
      boundBone {boneName=dyn_LeftSleeve_r}
      boxExtents=<20 30 30>
      constraintSetup {angularLimitsMax=<0 2 0>, angularLimitsMin=<0 -2 -10>, angularTarget=<1 0 0>}
      localJointOffset=<2 0 0>
     }]
    simSpaceSettings {maxAngularAcceleration=14, maxAngularVelocity=7, simSpaceAngularAlpha=0.5}
   }
   n26 = Anim Dynamics - Bone: dyn_ribben_a_l(ComponentPose=@n25, Alpha=1, ExternalForce="0,0,0") {
    angularDampingOverride=0.9999
    angularSpringConstant=5000
    bAngularSpring=true
    bOverrideAngularDamping=true
    bUseGravityOverride=true
    gravityScale=0.35
    simulationSpace=RootRelative
    boundBone {boneName=dyn_ribben_a_l}
    componentAppliedLinearAccClamp=<1500 1500 1500>
    componentLinearAccScale=<1 1 1>
    componentLinearVelScale=<1 1 1>
    gravityOverride {z=-980}
    physicsBodyDefinitions [{
      boundBone {boneName=dyn_ribben_a_l}
      boxExtents=<20 30 30>
      constraintSetup {angularLimitsMax=<0 5 15>, angularLimitsMin=<0 -15 -15>, angularTarget=<1 0 0>}
      localJointOffset=<4 0 0>
     }]
    simSpaceSettings {maxAngularAcceleration=14, maxAngularVelocity=7, simSpaceAngularAlpha=1}
   }
   n27 = Anim Dynamics - Bone: dyn_ribben_b_l(ComponentPose=@n26, Alpha=1, ExternalForce="0,0,0") {
    angularDampingOverride=0.9999
    angularSpringConstant=5000
    bAngularSpring=true
    bOverrideAngularDamping=true
    bUseGravityOverride=true
    gravityScale=0.35
    simulationSpace=RootRelative
    boundBone {boneName=dyn_ribben_b_l}
    componentAppliedLinearAccClamp=<1500 1500 1500>
    componentLinearAccScale=<1 1 1>
    componentLinearVelScale=<1 1 1>
    gravityOverride {z=-980}
    physicsBodyDefinitions [{
      boundBone {boneName=dyn_ribben_b_l}
      boxExtents=<20 30 30>
      constraintSetup {angularLimitsMax=<0 5 15>, angularLimitsMin=<0 -15 -15>, angularTarget=<1 0 0>}
      localJointOffset=<4 0 0>
     }]
    simSpaceSettings {maxAngularAcceleration=14, maxAngularVelocity=7, simSpaceAngularAlpha=1}
   }
   n28 = Anim Dynamics - Bone: dyn_ribben_a_r(ComponentPose=@n27, Alpha=1, ExternalForce="0,0,0") {
    angularDampingOverride=0.9999
    angularSpringConstant=5000
    bAngularSpring=true
    bOverrideAngularDamping=true
    bUseGravityOverride=true
    gravityScale=0.35
    simulationSpace=RootRelative
    boundBone {boneName=dyn_ribben_a_r}
    componentAppliedLinearAccClamp=<1500 1500 1500>
    componentLinearAccScale=<1 1 1>
    componentLinearVelScale=<1 1 1>
    gravityOverride {z=-980}
    physicsBodyDefinitions [{
      boundBone {boneName=dyn_ribben_a_r}
      boxExtents=<20 30 30>
      constraintSetup {angularLimitsMax=<0 5 15>, angularLimitsMin=<0 -15 -15>, angularTarget=<1 0 0>}
      localJointOffset=<4 0 0>
     }]
    simSpaceSettings {maxAngularAcceleration=14, maxAngularVelocity=7, simSpaceAngularAlpha=1}
   }
   n29 = Anim Dynamics - Bone: dyn_ribben_b_r(ComponentPose=@n28, Alpha=1, ExternalForce="0,0,0") {
    angularDampingOverride=0.9999
    angularSpringConstant=5000
    bAngularSpring=true
    bOverrideAngularDamping=true
    bUseGravityOverride=true
    gravityScale=0.35
    simulationSpace=RootRelative
    boundBone {boneName=dyn_ribben_b_r}
    componentAppliedLinearAccClamp=<1500 1500 1500>
    componentLinearAccScale=<1 1 1>
    componentLinearVelScale=<1 1 1>
    gravityOverride {z=-980}
    physicsBodyDefinitions [{
      boundBone {boneName=dyn_ribben_b_r}
      boxExtents=<20 30 30>
      constraintSetup {angularLimitsMax=<0 5 15>, angularLimitsMin=<0 -15 -15>, angularTarget=<1 0 0>}
      localJointOffset=<4 0 0>
     }]
    simSpaceSettings {maxAngularAcceleration=14, maxAngularVelocity=7, simSpaceAngularAlpha=1}
   }
   n30 = Anim Dynamics - Bone: dyn_knot_a(ComponentPose=@n29, Alpha=1, ExternalForce="0,0,0") {
    angularDampingOverride=1
    angularSpringConstant=15000
    bAngularSpring=true
    bOverrideAngularDamping=true
    bUseGravityOverride=true
    bUseSphericalLimits=true
    gravityScale=0.35
    simulationSpace=RootRelative
    boundBone {boneName=dyn_knot_a}
    componentAppliedLinearAccClamp=<2000 2000 2000>
    componentLinearAccScale=<1 1 1>
    componentLinearVelScale=<1 1 1>
    gravityOverride {z=-980}
    physicsBodyDefinitions [{
      boundBone {boneName=dyn_knot_a}
      boxExtents=<20 30 30>
      constraintSetup {angularLimitsMax=<0 20 20>, angularLimitsMin=<0 -20 -20>, angularTarget=<1 0 0>}
      localJointOffset=<3 0 0>
     }]
   }
   n31 = Anim Dynamics - Bone: dyn_knot_a_01(ComponentPose=@n30, Alpha=1, ExternalForce="0,0,0") {
    angularDampingOverride=1
    angularSpringConstant=15000
    bAngularSpring=true
    bOverrideAngularDamping=true
    bUseGravityOverride=true
    bUseSphericalLimits=true
    gravityScale=0.35
    simulationSpace=RootRelative
    boundBone {boneName=dyn_knot_a_01}
    componentAppliedLinearAccClamp=<2000 2000 2000>
    componentLinearAccScale=<1 1 1>
    componentLinearVelScale=<1 1 1>
    gravityOverride {z=-980}
    physicsBodyDefinitions [{
      boundBone {boneName=dyn_knot_a_01}
      boxExtents=<20 30 30>
      constraintSetup {angularLimitsMax=<0 20 20>, angularLimitsMin=<0 -20 -20>, angularTarget=<1 0 0>}
      localJointOffset=<3 0 0>
     }]
   }
   n32 = Anim Dynamics - Bone: dyn_knot_b(ComponentPose=@n31, Alpha=1, ExternalForce="0,0,0") {
    angularDampingOverride=1
    angularSpringConstant=15000
    bAngularSpring=true
    bOverrideAngularDamping=true
    bUseGravityOverride=true
    bUseSphericalLimits=true
    gravityScale=0.35
    simulationSpace=RootRelative
    boundBone {boneName=dyn_knot_b}
    componentAppliedLinearAccClamp=<2000 2000 2000>
    componentLinearAccScale=<1 1 1>
    componentLinearVelScale=<1 1 1>
    gravityOverride {z=-980}
    physicsBodyDefinitions [{
      boundBone {boneName=dyn_knot_b}
      boxExtents=<20 30 30>
      constraintSetup {angularLimitsMax=<0 20 20>, angularLimitsMin=<0 -20 -20>, angularTarget=<1 0 0>}
      localJointOffset=<3 0 0>
     }]
   }
   n33 = Anim Dynamics - Bone: dyn_knot_b_01(ComponentPose=@n32, Alpha=1, ExternalForce="0,0,0") {
    angularDampingOverride=1
    angularSpringConstant=15000
    bAngularSpring=true
    bOverrideAngularDamping=true
    bUseGravityOverride=true
    bUseSphericalLimits=true
    gravityScale=0.35
    simulationSpace=RootRelative
    boundBone {boneName=dyn_knot_b_01}
    componentAppliedLinearAccClamp=<2000 2000 2000>
    componentLinearAccScale=<1 1 1>
    componentLinearVelScale=<1 1 1>
    gravityOverride {z=-980}
    physicsBodyDefinitions [{
      boundBone {boneName=dyn_knot_b_01}
      boxExtents=<20 30 30>
      constraintSetup {angularLimitsMax=<0 20 20>, angularLimitsMin=<0 -20 -20>, angularTarget=<1 0 0>}
      localJointOffset=<3 0 0>
     }]
   }
   Output Pose(Result=Component To Local(ComponentPose=@n33)) {layerGroup=None}
  ]}
 EventGraph {nodes [
   n1 = Event BlueprintUpdateAnimation()
   > Set Delta Time X(Delta Time X=@n1.DeltaTimeX)
   > n2 = Sequence!(execute) -> (then_0, then_1, then_2:Exec, then_3:Exec)
   > n3 = "Gravity Override State Function (Parameter Asset) (Nativized)"(
    CharacterPartAnimInstance=Self-Reference()
    GameTimeInSeconds=GetGameTimeInSeconds()
    Parameters=&/F_MED_LocalTalent_DaisyParameter
   )
   > Set GravityOverride(GravityOverride=@n3)
   > Set RBAN Settings Out(RBAN Settings Out=@n3.OutRigidBodyAnimNodeParameters)
   n4 = Get bIsMoving2D()
   > Set OnlyCrouch(OnlyCrouch:bool=AND Boolean(A=Get bIsCrouching(), B=NOT Boolean(A=@n4))) -> (Output_Get:bool)
   n5 = Get bIsCrouching()
   n6 = Get bIsMoving2D()
   Set CrouchWalk(execute=@n2.then_1, CrouchWalk:double="To Float (Boolean)"(InBool=AND Boolean(A=@n5, B=@n6))) -> (Output_Get:double)
   > n7 = PlayerMovement(AnimInstance=Self-Reference(), XDotProductCoefficient=1, YDotProductCoefficient=1, ZDotProductCoefficient=1)
   n8 = BreakVector(InVec=MakeVector(X=@n7, Y=@n7.YMovement))
   n9 = "Absolute (Float)"(A="Clamp (Float)"(Value=@n8, Min=-1, Max=0))
   > "Set Move Left "("Move Left ":double=@n9) -> (Output_Get:double)
   > Set Move Right(Move Right:double="Clamp (Float)"(Value=@n8, Min=0, Max=1)) -> (Output_Get:double)
   n10 = PlayerMovement(execute=@n2.then_2, AnimInstance=Self-Reference(), XDotProductCoefficient=100, YDotProductCoefficient=100, ZDotProductCoefficient=100)
   n11 = BreakVector(InVec=MakeVector(X=@n10, Y=@n10.YMovement))
   > Set moveLeft bool(moveLeft bool:bool="float < float"(A=@n11, B=-50)) -> (Output_Get:bool)
   > Set moveRight bool(moveRight bool:bool="float > float"(A=@n11, B=50)) -> (Output_Get:bool)
   n12 = AND Boolean!(
    A=Get bIsCrouching()
    B=Get bIsMoving2D()
    C:bool=OR Boolean(A=Get moveLeft bool() -> (moveLeft bool:bool), B=Get moveRight bool() -> (moveRight bool:bool))
   ) -> (ReturnValue)
   > Set isCrouchStrafe float(isCrouchStrafe float:double=FInterpTo(Current=Get isCrouchStrafe float(), Target="To Float (Boolean)"(InBool=@n12), DeltaTime=Get Delta Time X(), InterpSpeed=10)) -> (Output_Get:double)
   float + float(A=@n9, B=Get CrouchWalk() -> (CrouchWalk:double))
   n13 = GetOwnerSkeletalMeshComponent()
   Set Emo(Emo:bool=Montage_IsPlaying(self=GetAnimInstance(self=@n13), Montage=/BRCosmetics/Animation/Game/MainPlayer/Emotes/Eastern_Bloc/Emote_EasternBloc_CMF_M)) -> (Output_Get:bool)
  ]}
]
variables [
 CrouchWalk {type=number}
 Delta Time X {type=number}
 Emo {type=boolean}
 GravityOverride {type=Vector}
 isCrouchStrafe float {type=number}
 "Move Left " {type=number}
 Move Right {type=number}
 moveLeft bool {type=boolean}
 moveRight bool {type=boolean}
 OnlyCrouch {type=boolean}
 RBAN Settings Out {type=FortRigidBodyAnimNodeParameters}
]
