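# Physics asset for the Daisy LocalTalent female medium body, rigging four dangling loop chains as capsule bodies around the pelvis and thighs.
# Pelvis, thigh and thigh-twist capsules form the body volume, with limited thigh joints and free twist joints.
# Loop bases attach to the pelvis, locked in rotation on the left and free on the right.
# Each dynamic chain is linked segment to segment with tight twist limits, with narrower swing on the right.
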PhysicsAsset /BRCosmetics/Characters/Player/Female/Medium/Bodies/F_MED_LocalTalent_Daisy/Meshes/F_MED_LocalTalent_Daisy_Physics
bodies {
 base_loop_a_l [{shapeType=Capsule, length=0.2, radius=2, shapeName=base_loop_a_l_capsule, rotation=<rot 0 -90 -90>}]
 base_loop_a_r [{shapeType=Capsule, length=0.505, radius=0.505, shapeName=base_loop_a_r_capsule, rotation=<rot 0 -90 0>}]
 base_loop_b_l [{shapeType=Capsule, length=0.2, radius=2, shapeName=base_loop_b_l_capsule, rotation=<rot 0 -90 -90>}]
 base_loop_b_r [{shapeType=Capsule, length=0.2, radius=3, shapeName=base_loop_b_r_capsule, rotation=<rot 0 -90 -90>}]
 dyn_loop_a_l_01 [{shapeType=Capsule, length=2, radius=2, shapeName=dyn_loop_a_l_01_capsule, center=<2 0 0>, rotation=<rot -90 0 0>}]
 dyn_loop_a_l_02 [{shapeType=Capsule, length=2, radius=2, shapeName=dyn_loop_a_l_02_capsule, center=<2 0 0>, rotation=<rot -90 0 0>}]
 dyn_loop_a_l_03 [{shapeType=Capsule, length=2, radius=2, shapeName=dyn_loop_a_l_03_capsule, center=<2 0 0>, rotation=<rot -90 0 0>}]
 dyn_loop_a_r_01 [{shapeType=Capsule, length=4, radius=2, shapeName=dyn_loop_a_r_01_capsule, center=<3 0 -1.5>, rotation=<rot 0 90 90>}]
 dyn_loop_a_r_02 [{shapeType=Capsule, length=4, radius=2, shapeName=dyn_loop_a_r_02_capsule, center=<3 0 -1.2>, rotation=<rot 0 90 100>}]
 dyn_loop_a_r_03 [{shapeType=Capsule, length=4, radius=2, shapeName=dyn_loop_a_r_03_capsule, center=<2 0 -1>, rotation=<rot 0 90 90>}]
 dyn_loop_a_r_04 [{shapeType=Capsule, length=5, radius=3, shapeName=dyn_loop_a_r_04_capsule, center=<2 0 -1>, rotation=<rot 0.225 82.8602 84.7275>}]
 dyn_loop_b_l_01 [{shapeType=Capsule, length=2, radius=2, shapeName=dyn_loop_b_l_01_capsule, center=<2 0 0>, rotation=<rot -90 0 0>}]
 dyn_loop_b_l_02 [{shapeType=Capsule, length=2, radius=2, shapeName=dyn_loop_b_l_02_capsule, center=<2 0 0>, rotation=<rot -90 0 0>}]
 dyn_loop_b_l_03 [{shapeType=Capsule, length=2, radius=2, shapeName=dyn_loop_b_l_03_capsule, center=<2 0 0>, rotation=<rot -90 0 0>}]
 dyn_loop_b_l_04 [{shapeType=Capsule, length=2, radius=2, shapeName=dyn_loop_b_l_04_capsule, center=<2 0 0>, rotation=<rot -90 0 0>}]
 dyn_loop_b_r_01 [{shapeType=Capsule, length=5, radius=2, shapeName=dyn_loop_b_r_01_capsule, center=<2 0 -1.9>, rotation=<rot 0 90 90>}]
 dyn_loop_b_r_02 [{shapeType=Capsule, length=5, radius=2.2, shapeName=dyn_loop_b_r_02_capsule, center=<3 0.2 -1.9>, rotation=<rot 0 90 94>}]
 dyn_loop_b_r_03 [{shapeType=Capsule, length=5, radius=2.5, shapeName=dyn_loop_b_r_03_capsule, center=<2 0 -1.5>, rotation=<rot 0 90 90>}]
 pelvis [{shapeType=Capsule, length=10, radius=12, center=<-8.2777 -2.093 -0.0683>, rotation=<rot 0 -90 0>}]
 thigh_l [
  {shapeType=Capsule, length=34.3283, radius=8, center=<-17.6554 0.0511 0.4031>, rotation=<rot 4.5377 85.4335 -86.6138>}
  {
   shapeType=Capsule
   length=1.153
   radius=6.5
   shapeName=thigh_l_capsule
   center=<-22.7951 0.8374 -8.6285>
   rotation=<rot -83.7634 179.9862 179.993>
  }
 ]
 thigh_r [{shapeType=Capsule, length=34.3283, radius=8, center=<17.6554 -0.0511 -0.4031>, rotation=<rot 4.5377 85.4336 92.3174>}]
 thigh_twist_01_l [{
   shapeType=Capsule
   length=5.9126
   radius=11.3895
   shapeName=thigh_twist_01_l_capsule
   center=<21.0584 -1.4311 3.4641>
   rotation=<rot -12.8835 -84.9419 -101.2986>
  }]
 thigh_twist_01_r [{
   shapeType=Capsule
   length=5.9126
   radius=11.3895
   shapeName=thigh_twist_01_r_capsule
   center=<-21.0583 0.5778 -3.3773>
   rotation=<rot -12.8835 -84.9422 76.4508>
  }]
}
constraints [
 {bone1Name=thigh_l, bone2Name=pelvis, swing1LimitDegrees=45, swing2LimitDegrees=45, twistLimitDegrees=45}
 {bone1Name=thigh_r, bone2Name=pelvis, swing1LimitDegrees=45, swing2LimitDegrees=45, twistLimitDegrees=45}
 {
  bone1Name=base_loop_b_l
  bone2Name=pelvis
  swing1LimitDegrees=45
  swing1Motion=Locked
  swing2LimitDegrees=45
  swing2Motion=Locked
  twistLimitDegrees=45
  twistMotion=Locked
 }
 {
  bone1Name=base_loop_a_l
  bone2Name=pelvis
  swing1LimitDegrees=45
  swing1Motion=Locked
  swing2LimitDegrees=45
  swing2Motion=Locked
  twistLimitDegrees=45
  twistMotion=Locked
 }
 {
  bone1Name=base_loop_b_r
  bone2Name=pelvis
  swing1LimitDegrees=45
  swing1Motion=Free
  swing2LimitDegrees=45
  swing2Motion=Free
  twistLimitDegrees=45
  twistMotion=Free
 }
 {
  bone1Name=base_loop_a_r
  bone2Name=pelvis
  swing1LimitDegrees=45
  swing1Motion=Free
  swing2LimitDegrees=45
  swing2Motion=Free
  twistLimitDegrees=45
  twistMotion=Free
 }
 {bone1Name=dyn_loop_a_r_01, bone2Name=base_loop_a_r, swing1LimitDegrees=20, swing2LimitDegrees=30, twistLimitDegrees=2}
 {bone1Name=dyn_loop_a_r_02, bone2Name=dyn_loop_a_r_01, swing1LimitDegrees=20, swing2LimitDegrees=30, twistLimitDegrees=2}
 {bone1Name=dyn_loop_a_r_03, bone2Name=dyn_loop_a_r_02, swing1LimitDegrees=20, swing2LimitDegrees=30, twistLimitDegrees=2}
 {bone1Name=dyn_loop_a_r_04, bone2Name=dyn_loop_a_r_03, swing1LimitDegrees=20, swing2LimitDegrees=30, twistLimitDegrees=2}
 {bone1Name=dyn_loop_b_r_01, bone2Name=base_loop_b_r, swing1LimitDegrees=20, swing2LimitDegrees=30, twistLimitDegrees=2}
 {bone1Name=dyn_loop_b_r_02, bone2Name=dyn_loop_b_r_01, swing1LimitDegrees=20, swing2LimitDegrees=30, twistLimitDegrees=2}
 {bone1Name=dyn_loop_b_r_03, bone2Name=dyn_loop_b_r_02, swing1LimitDegrees=20, swing2LimitDegrees=30, twistLimitDegrees=2}
 {
  bone1Name=thigh_twist_01_l
  bone2Name=thigh_l
  swing1LimitDegrees=45
  swing1Motion=Free
  swing2LimitDegrees=45
  swing2Motion=Free
  twistLimitDegrees=45
  twistMotion=Free
 }
 {
  bone1Name=thigh_twist_01_r
  bone2Name=thigh_r
  swing1LimitDegrees=45
  swing1Motion=Free
  swing2LimitDegrees=45
  swing2Motion=Free
  twistLimitDegrees=45
  twistMotion=Free
 }
 {bone1Name=dyn_loop_a_l_01, bone2Name=base_loop_a_l, swing1LimitDegrees=45, swing2LimitDegrees=45, twistLimitDegrees=2}
 {bone1Name=dyn_loop_a_l_02, bone2Name=dyn_loop_a_l_01, swing1LimitDegrees=45, swing2LimitDegrees=45, twistLimitDegrees=2}
 {bone1Name=dyn_loop_a_l_03, bone2Name=dyn_loop_a_l_02, swing1LimitDegrees=45, swing2LimitDegrees=45, twistLimitDegrees=2}
 {bone1Name=dyn_loop_b_l_01, bone2Name=base_loop_b_l, swing1LimitDegrees=45, swing2LimitDegrees=45, twistLimitDegrees=2}
 {bone1Name=dyn_loop_b_l_02, bone2Name=dyn_loop_b_l_01, swing1LimitDegrees=45, swing2LimitDegrees=45, twistLimitDegrees=2}
 {bone1Name=dyn_loop_b_l_03, bone2Name=dyn_loop_b_l_02, swing1LimitDegrees=45, swing2LimitDegrees=45, twistLimitDegrees=2}
 {bone1Name=dyn_loop_b_l_04, bone2Name=dyn_loop_b_l_03, swing1LimitDegrees=45, swing2LimitDegrees=45, twistLimitDegrees=2}
]
