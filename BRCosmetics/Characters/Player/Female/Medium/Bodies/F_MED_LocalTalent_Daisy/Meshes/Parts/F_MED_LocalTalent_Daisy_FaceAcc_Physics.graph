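# Physics asset for the face accessory mesh of the female character LocalTalent Daisy, set up for a six-bone dynamic ponytail.
# Capsules on the head, neck, upper spine, clavicles and upper arms form the body volumes, linked by free constraints.
# The ponytail root is locked to the head.
# Later links allow almost no twist, with asymmetric swing limits that widen toward the tip.
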PhysicsAsset /BRCosmetics/Characters/Player/Female/Medium/Bodies/F_MED_LocalTalent_Daisy/Meshes/Parts/F_MED_LocalTalent_Daisy_FaceAcc_Physics
bodies {
 clavicle_l [{shapeType=Capsule, length=12.6283, radius=7.5, shapeName=clavicle_l_capsule, center=<5 0 -1.6229>, rotation=<rot 0 -90 90>}]
 clavicle_r [{shapeType=Capsule, length=12.2586, radius=7.5, shapeName=clavicle_r_capsule, center=<-5 0 1.6229>, rotation=<rot 0 -90 -90>}]
 dyn_ponytail_01 [{shapeType=Capsule, length=2, radius=2, shapeName=dyn_ponytail_01_capsule, rotation=<rot -90 0 0>}]
 dyn_ponytail_02 [{shapeType=Capsule, length=7.8966, radius=5, shapeName=dyn_ponytail_02_capsule, center=<5 0 0>, rotation=<rot -90 0 0>}]
 dyn_ponytail_03 [{shapeType=Capsule, length=8.6213, radius=5, shapeName=dyn_ponytail_03_capsule, center=<5 0 0>, rotation=<rot -90 0 0>}]
 dyn_ponytail_04 [{shapeType=Capsule, length=8.9016, radius=5, shapeName=dyn_ponytail_04_capsule, center=<5 0 0>, rotation=<rot -90 0 0>}]
 dyn_ponytail_05 [{shapeType=Capsule, length=6.8417, radius=5, shapeName=dyn_ponytail_05_capsule, center=<5 0 0>, rotation=<rot -90 0 0>}]
 dyn_ponytail_06 [{shapeType=Capsule, length=3, radius=4, shapeName=dyn_ponytail_06_capsule, center=<1.7522 0 0>, rotation=<rot -90 0 0>}]
 head [{shapeType=Capsule, length=8, radius=8, center=<3.4471 0.4526 0>, rotation=<rot 0 -90 -90>}]
 neck_01 [{shapeType=Capsule, length=4.3856, radius=5.6144, shapeName=neck_01_capsule, rotation=<rot -90 0 0>}]
 neck_02 [{shapeType=Capsule, length=4.3856, radius=5.6144, shapeName=neck_02_capsule, rotation=<rot -90 0 0>}]
 spine_03 [{shapeType=Capsule, length=20.0893, radius=9.9107, shapeName=spine_03_capsule}]
 spine_04 [{shapeType=Capsule, length=20.193, radius=9.807, shapeName=spine_04_capsule}]
 spine_05 [{shapeType=Capsule, length=20.2336, radius=9.7664, shapeName=spine_05_capsule}]
 upperarm_l [{shapeType=Capsule, length=18.9692, radius=6.3198, shapeName=upperarm_l_capsule, center=<10 0 0>, rotation=<rot 0 -90 90>}]
 upperarm_r [{shapeType=Capsule, length=18.7079, radius=6.5811, shapeName=upperarm_r_capsule, center=<-10 -1.0716 0>, rotation=<rot 0 -90 -90>}]
}
constraints [
 {
  bone1Name=spine_04
  bone2Name=spine_03
  swing1LimitDegrees=45
  swing1Motion=Free
  swing2LimitDegrees=45
  swing2Motion=Free
  twistLimitDegrees=45
  twistMotion=Free
 }
 {
  bone1Name=spine_05
  bone2Name=spine_04
  swing1LimitDegrees=45
  swing1Motion=Free
  swing2LimitDegrees=45
  swing2Motion=Free
  twistLimitDegrees=45
  twistMotion=Free
 }
 {
  bone1Name=head
  bone2Name=neck_02
  swing1LimitDegrees=45
  swing1Motion=Free
  swing2LimitDegrees=45
  swing2Motion=Free
  twistLimitDegrees=45
  twistMotion=Free
 }
 {
  bone1Name=clavicle_l
  bone2Name=spine_05
  swing1LimitDegrees=45
  swing1Motion=Free
  swing2LimitDegrees=45
  swing2Motion=Free
  twistLimitDegrees=45
  twistMotion=Free
 }
 {
  bone1Name=clavicle_r
  bone2Name=spine_05
  swing1LimitDegrees=45
  swing1Motion=Free
  swing2LimitDegrees=45
  swing2Motion=Free
  twistLimitDegrees=45
  twistMotion=Free
 }
 {
  bone1Name=upperarm_l
  bone2Name=clavicle_l
  swing1LimitDegrees=45
  swing1Motion=Free
  swing2LimitDegrees=45
  swing2Motion=Free
  twistLimitDegrees=45
  twistMotion=Free
 }
 {
  bone1Name=upperarm_r
  bone2Name=clavicle_r
  swing1LimitDegrees=45
  swing1Motion=Free
  swing2LimitDegrees=45
  swing2Motion=Free
  twistLimitDegrees=45
  twistMotion=Free
 }
 {
  bone1Name=dyn_ponytail_01
  bone2Name=head
  swing1LimitDegrees=45
  swing1Motion=Locked
  swing2LimitDegrees=45
  swing2Motion=Locked
  twistLimitDegrees=45
  twistMotion=Locked
 }
 {bone1Name=dyn_ponytail_02, bone2Name=dyn_ponytail_01, swing1LimitDegrees=10, swing2LimitDegrees=30, twistLimitDegrees=1}
 {bone1Name=dyn_ponytail_03, bone2Name=dyn_ponytail_02, swing1LimitDegrees=10, swing2LimitDegrees=30, twistLimitDegrees=1}
 {bone1Name=dyn_ponytail_04, bone2Name=dyn_ponytail_03, swing1LimitDegrees=15, swing2LimitDegrees=35, twistLimitDegrees=1}
 {bone1Name=dyn_ponytail_05, bone2Name=dyn_ponytail_04, swing1LimitDegrees=20, swing2LimitDegrees=40, twistLimitDegrees=1}
 {bone1Name=dyn_ponytail_06, bone2Name=dyn_ponytail_05, swing1LimitDegrees=20, swing2LimitDegrees=45, twistLimitDegrees=1}
 {
  bone1Name=neck_01
  bone2Name=spine_05
  swing1LimitDegrees=45
  swing1Motion=Free
  swing2LimitDegrees=45
  swing2Motion=Free
  twistLimitDegrees=45
  twistMotion=Free
 }
 {
  bone1Name=neck_02
  bone2Name=neck_01
  swing1LimitDegrees=45
  swing1Motion=Free
  swing2LimitDegrees=45
  swing2Motion=Free
  twistLimitDegrees=45
  twistMotion=Free
 }
]
